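# Dopey_Res_Cue is an event sound cue that plays one of four Dopey "Res" wave variations at random with equal weighting.
# A modulator sits over the random selection, and the result plays louder than default through a pickaxe player-impact attenuation and a quieter ducking impact sound class.
# A concurrency override caps it at two simultaneous instances, blocking new plays and scaling volume.
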SoundCue /Figment_S09/Athena/Sounds/Events/Dopey/Dopey_Res_Cue
root=Modulator_0
nodes [
 {id=Modulator_0, children [Random_1]}
 {id=Random_1, children [WavePlayer_8, WavePlayer_9, WavePlayer_10, WavePlayer_11], properties {weights [1,1,1,1]}}
 {id=WavePlayer_10, properties {soundWaveAssetPtr=&/Dopey_Res_02}}
 {id=WavePlayer_11, properties {soundWaveAssetPtr=&/Dopey_Res_03}}
 {id=WavePlayer_8, properties {soundWaveAssetPtr=&/Dopey_Res_04}}
 {id=WavePlayer_9, properties {soundWaveAssetPtr=&/Dopey_Res_01}}
]
properties {
 attenuationSettings=/Game/Sounds/Attenuation/Fort_Weapon_Pickaxe_Player_Impact_Att
 bOverrideConcurrency=true
 maxDistance=3512
 soundClassObject=/Game/Sounds/FortSoundClassesAndModes/Fort_ImpactSounds_DuckDefaultQuieter_PSM
 volumeMultiplier=1.5
 concurrencyOverrides {maxCount=2, resolutionRule=PreventNew, volumeScale=0.8}
}
